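# Translucent unlit Niagara sprite material for an explosion streak, filed under the Zeus Lightning item's effects.
# Emissive tints a vertical UV gradient, scaled by dynamic BottomBright, with particle color through MF_FX_EmissiveAdjustments.
# Opacity combines streak texture and particle alpha with optional depth fade (bypassed on mobile) and near-camera fade.
# A noise texture plus inverted V gradient, offset by dynamic Dissolve_Value, erodes the streak along its length.
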Material /SunRoseZeusGameplay/Items/Lightning/FX/Explosion/M_ExplosionStreak_02
graph {
 properties {
  bAllowNegativeEmissiveColor=true
  bUsedWithNanite=true
  bUsedWithNiagaraSprites=true
  blendMode=BLEND_Translucent
  pixelDepthOffsetMode=PDOM_Legacy
  shadingModel=MSM_Unlit
  usedShadingModels=Unlit
 }
 nodes [
  n1 = TextureCoordinate()
  n2 = DynamicParameter() {defaultValue=<rgba 1 1 10 0>, paramNames [DepthFade, Dissolve_Value, BP_HueShift, BottomBright]}
  n3 = Multiply(
   A=Power(Base=MaterialFunctionCall(Float2=@n1) {materialFunction=&BreakOutFloat2Components}.G)
   B=@n2.BottomBright
  )
  n4 = ParticleColor()
  MaterialFunctionCall(Emissive=Multiply(A=Add(A=@n3), B=@n4.RGB)) {materialFunction=&MF_FX_EmissiveAdjustments}
  n5 = Multiply(A=@n4.A, B=TextureSample() {texture=/Game/Effects/Fort_Effects/Gameplay/SmokeDustExplosions/Textures/T_Explosion_Streak_02}.RGB)
  n6 = LinearInterpolate(
   A=ScalarParameter() {defaultValue=5, group=Opacity|DepthFade, parameterName=DepthFade_Distance, sortPriority=2}
   B=DynamicParameter() {defaultValue=<rgba 1 1 10 0>, paramNames [DepthFade, Dissolve_Value, BP_HueShift, BottomBright]}.DepthFade
   Alpha=ScalarParameter() {group=Opacity|DepthFade, parameterName=Use Dynamic Parameter to Control Depth Fade, sortPriority=1}
  )
  n7 = LinearInterpolate(
   A=@n5
   B=ShadingPathSwitch(Default=DepthFade(Opacity=@n5, FadeDistance=@n6), Mobile=@n5)
   Alpha=ScalarParameter() {group=Opacity|DepthFade, parameterName=Use Depth Fade}
  )
  n8 = ScalarParameter() {defaultValue=512, group=Opacity|Near_Camera_Fade, parameterName=NearCamFade_Distance, sortPriority=1}
  n9 = Clamp(None=Power(
    Base=MaterialFunctionCall(Distance=@n8) {materialFunction=&Near_Camera_Fade}.Result
    Exp=ScalarParameter() {defaultValue=1, group=Opacity|Near_Camera_Fade, parameterName=NearCamFade_Power, sortPriority=2}
   )) {minDefault=0.02}
  n10 = OneMinus(None=MaterialFunctionCall(Float2=TextureCoordinate()) {materialFunction=&BreakOutFloat2Components}.G)
  n11 = VectorParameter() {parameterName=Dissolve_V, defaultValue=<rgba 0 2 2 1>}
  n12 = TextureObjectParameter() {
   group=Dissolve
   parameterName=Dissolve_Texture
   samplerType=SAMPLERTYPE_LinearGrayscale
   sortPriority=0
   texture=/Game/Characters/Player/Male/Medium/Bodies/M_MED_Citadel/FX/T_Citadel_Noise_Dissolve_Handpaint
  }
  n13 = VectorParameter() {group=Dissolve, parameterName=Dissolve_UVControl, channelNames {a=VPan, b=UPan, g=VScale, r=UScale}, defaultValue=<rgba 2 0.5 0 0>}
  n14 = Add(
   A=Multiply(A=Add(A=@n10, B=@n2.Dissolve_Value), B=@n11.B)
   B=Subtract(
    A=MaterialFunctionCall(Texture=@n12, UVControl_RGB=@n13, UVControl_A=@n13) {materialFunction=&FXCommon_TextureSample}.R
    B=@n11.R
   )
  )
  Multiply(
   A=LinearInterpolate(A=@n7, B=Multiply(A=@n7, B=@n9), Alpha=ScalarParameter() {group=Opacity|Near_Camera_Fade, parameterName=Add Near Camera Fade})
   B=Saturate(None=Multiply(A=@n14, B=@n11.G))
  )
 ]
}
